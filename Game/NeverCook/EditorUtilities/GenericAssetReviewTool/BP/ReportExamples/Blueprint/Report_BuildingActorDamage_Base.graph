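# Editor utility report for the Generic Asset Review Tool that flags building actor Blueprints with CanBeDamaged disabled.
# It matches each asset's NativeParentClass tag exactly against a list of wall, floor, stairs, roof and prop classes.
# For matches, inline Python reads the damage flag from the generated class's CDO.
# A disabled flag adds an error recommending BuildingStaticGeo, and Valid reflects whether the error string is empty.
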EditorUtilityBlueprint /Game/NeverCook/EditorUtilities/GenericAssetReviewTool/BP/ReportExamples/Blueprint/Report_BuildingActorDamage_Base
graphs [
 CheckIfBuildingActorClass {nodes [
   CheckIfBuildingActorClass!() -> (then, AssetData:Asset Data Structure) {is=title K2Node_FunctionEntry}
   > n1 = For Each Loop with Break!(
    Exec:Exec
    Array:Array of Object Soft Class References=Get BuildingActorClasses() -> (BuildingActorClasses:Array of Object Soft Class References)
    Break:Exec=@n4.then
   ) -> (LoopBody:Exec, Array Element:Object Soft Class Reference, Array Index:int, Completed:Exec)
   n2 = MakeSoftClassPath(PathString=GetTagValue(InAssetData=Get AssetData(), InTagName=MakeLiteralName(Value=NativeParentClass)))
   n3 = "Equal Exactly (String)"(
    A="To String (SoftClassReference)"!(SoftClassReference:"Object Soft Class Reference (by ref)"=To Soft Class Reference(SoftClassPath=@n2)) -> (ReturnValue:String)
    B="To String (SoftClassReference)"(SoftClassReference=@n1)
   )
   Return Node!(execute=@n1.Completed, IsBuildngActorClass:bool=@n3)
   > n4 = Branch(Condition=@n3)
  ]}
 EventGraph {links [], nodes []}
 RunFilterLogicOnAssetData {nodes [
   n1 = RunFilterLogicOnAssetData()
   > n2 = CheckIfBuildingActorClass!(execute, self, AssetData:Asset Data Structure=@n1) -> (then, IsBuildngActorClass:bool) {is=title K2Node_CallFunction}
   > n3 = Branch(Condition=@n2)
   > n4 = Sequence()
   > n5 = Execute Python Script!(
    execute
    PythonScript="import unreal\r\n\r\nassetObject = unreal.load_object(None, asset_object_path)\r\n\r\ncdo = unreal.get_default_object(assetObject)\r\n\r\ndamage_state = cdo.get_editor_property(\"bcanbedamaged\")"
    asset_object_path:String=Append(A=ToSoftObjectPath!(InAssetData:"Asset Data Structure (by ref)"=@n1) -> (ReturnValue_PathString:String), B=_C)
   ) -> (then, ReturnValue, damage_state:bool=~) {inputs [asset_object_path], outputs [damage_state]}
   > n6 = Branch(Condition=@n5.damage_state)
   > Return Node!(execute, Valid:bool, ReportData:Struct GART Report Data Structure)
   AddError(execute=@n6.else, NewError="CanBeDamaged flag disabled, indestrutible assets must use BuildingStaticGeo parent class")
   Return Node(
    execute=@n4.then_1
    Valid=IsEmpty(InString=GetCurrentErrorString!() -> (CurrentErrorString:String))
    ReportData=Make Struct GART Report Data(
     DisplayTitle_20_#1=Building Actor
     DisplayDescShort_29_#2=CanBeDamaged set to 'False'.
     DisplayColorOverride_40_#3="Lerp (LinearColor)"(A=(R=0.7552,G=0.2199,B=0,A=1), Alpha=0)
    )
   )
   Return Node(execute=@n3.else)
  ]}
]
variables [BuildingActorClasses {
  type="array<Class>"
  default [
   &BuildingWall
   &BuildingFloor
   &BuildingStairs
   &BuildingRoof
   &BuildingProp
  ]
 }]
