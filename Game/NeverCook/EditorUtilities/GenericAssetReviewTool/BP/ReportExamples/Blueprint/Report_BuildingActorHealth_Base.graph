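# Editor utility report for the Generic Asset Review Tool that flags building actor Blueprints whose health attribute init category is 'None', which makes them indestructible.
# It filters assets by matching NativeParentClass against wall, floor, stairs, roof and prop classes.
# An embedded Python script reads the class default object's init categories, resource type, amount row and building type.
# From these it adds a nested error recommending Prop, Resource or LDBuildingNormal settings.
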EditorUtilityBlueprint /Game/NeverCook/EditorUtilities/GenericAssetReviewTool/BP/ReportExamples/Blueprint/Report_BuildingActorHealth_Base
graphs [
 CheckIfBuildingActorClass {nodes [
   CheckIfBuildingActorClass!() -> (then, AssetData:Asset Data Structure) {is=title K2Node_FunctionEntry}
   > n1 = For Each Loop with Break!(
    Exec:Exec
    Array:Array of Object Soft Class References=Get BuildingActorClasses() -> (BuildingActorClasses:Array of Object Soft Class References)
    Break:Exec=@n4.then
   ) -> (LoopBody:Exec, Array Element:Object Soft Class Reference, Array Index:int, Completed:Exec)
   n2 = MakeSoftClassPath(PathString=GetTagValue(InAssetData=Get AssetData(), InTagName=MakeLiteralName(Value=NativeParentClass)))
   n3 = "Equal Exactly (String)"(
    A="To String (SoftClassReference)"!(SoftClassReference:"Object Soft Class Reference (by ref)"=To Soft Class Reference(SoftClassPath=@n2)) -> (ReturnValue:String)
    B="To String (SoftClassReference)"(SoftClassReference=@n1)
   )
   Return Node!(execute=@n1.Completed, IsBuildngActorClass:bool=@n3)
   > n4 = Branch(Condition=@n3)
  ]}
 EventGraph {links [], nodes []}
 RunFilterLogicOnAssetData {nodes [
   n1 = RunFilterLogicOnAssetData()
   > n2 = CheckIfBuildingActorClass!(execute, self, AssetData:Asset Data Structure=@n1) -> (then, IsBuildngActorClass:bool) {is=title K2Node_CallFunction}
   > n3 = Branch(Condition=@n2)
   > n4 = Sequence()
   > n5 = Execute Python Script!(
    execute
    PythonScript="import unreal\r\n\r\nassetObject = unreal.load_object(None, asset_object_path)\r\n\r\ncdo = unreal.get_default_object(assetObject)\r\n\r\nattInitKeys = cdo.get_editor_property(\"AttributeInitKeys\")\r\ncurrent_attribute_cat_0 = attInitKeys[0].get_editor_property(\"attributeInitCategory\")\r\ncurrent_attribute_cat_1 = attInitKeys[1].get_editor_property(\"attributeInitCategory\")\r\ncurrent_res_type = cdo.get_editor_property(\"ResourceType\").name\r\ncurrent_res_amount = cdo.get_editor_property(\"BuildingResourceAmountOverride\").get_editor_property(\"RowName\")\r\nbuilding_type_enum = str(cdo.get_editor_property(\"BuildingType\"))"
    asset_object_path:String=Append(A=ToSoftObjectPath!(InAssetData:"Asset Data Structure (by ref)"=@n1) -> (ReturnValue_PathString:String), B=_C)
   ) -> (
    then
    ReturnValue
    currentAttributeCat_0:String
    currentAttributeCat_1:String
    currentResType:String
    currentResAmount:String
    buildingTypeEnum:String
   ) {inputs [asset_object_path], outputs [currentAttributeCat_0, currentAttributeCat_1, currentResType, currentResAmount, buildingTypeEnum]}
   > n6 = Branch(Condition=OR Boolean(A="Equal Exactly (String)"(A=@n5.currentAttributeCat_0, B=None), B="Equal Exactly (String)"(A=@n5.currentAttributeCat_1, B=None)))
   > Set l_currentResType(l_currentResType:String=@n5.currentResType) -> (Output_Get:String)
   > Set l_currentResAmount(l_currentResAmount:String=@n5.currentResAmount) -> (Output_Get:String)
   > Set l_currentBuildingType(l_currentBuildingType:String=@n5.buildingTypeEnum) -> (Output_Get:String)
   > AddError(NewError="Attribute initialization catagories 'None' will cause actors to be indestructible. ")
   > n7 = Sequence()
   > n8 = Branch(Condition=Contains(SearchIn=Get l_currentBuildingType() -> (l_currentBuildingType:String), Substring=prop))
   > n9 = Branch(Condition="Equal Exactly (String)"(A=Get l_currentResType() -> (l_currentResType:String), B=None))
   > Set l_recomendedSetting(l_recomendedSetting:String="Without resource type set, an accurate setting cannot be reccommended for initilization catagory.") -> (Output_Get:String)
   n10 = Branch(execute=@n9.else, Condition=Contains(SearchIn=Get l_currentResAmount() -> (l_currentResAmount:String), Substring=prop))
   > Set l_recomendedSetting(l_recomendedSetting=Append!(
     A="Based on current resource type and amount, initilization catagory should be set to one of the Prop"
     B=Get l_currentResType()
     C:String=" settings."
    ) -> (ReturnValue))
   Set l_recomendedSetting(
    execute=@n10.else
    l_recomendedSetting=Append(
     A="Based on current resource type and amount, initilization catagory should be set to one of the Resource"
     B=Get l_currentResType()
     C=" settings."
    )
   )
   Set l_recomendedSetting(
    execute=@n8.else
    l_recomendedSetting=Append!(A="Based on building actor type, initilization catagory should be set to", B=" LDBuildingNormal.") -> (ReturnValue)
   )
   AddError(execute=@n7.then_1, NewError=Get l_recomendedSetting() -> (l_recomendedSetting:String), Depth=1)
   Return Node!(execute=@n6.else, Valid:bool, ReportData:Struct GART Report Data Structure)
   Return Node(
    execute=@n4.then_1
    Valid=IsEmpty(InString=GetCurrentErrorString!() -> (CurrentErrorString:String))
    ReportData=Make Struct GART Report Data(
     DisplayTitle_20_#1=Building Actor
     DisplayDescShort_29_#2=Health attribute set to 'None'.
     DisplayColorOverride_40_#3="Lerp (LinearColor)"(A=(R=0.7552,G=0.0328,B=0.0236,A=1), Alpha=0)
    )
   )
   Return Node(execute=@n3.else)
  ]}
]
variables [BuildingActorClasses {
  type="array<Class>"
  default [
   &BuildingWall
   &BuildingFloor
   &BuildingStairs
   &BuildingRoof
   &BuildingProp
  ]
 }]
