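# Editor utility report for the Generic Asset Review Tool that flags building actor Blueprints containing non-Static mobility mesh components.
# It matches the asset's NativeParentClass tag exactly against a list of building classes, avoiding an asset load for non-matches.
# Matching Blueprints are loaded, and SubobjectDataSubsystem gathers their non-editor-only StaticMeshComponents.
# Components not set to Static, including Stationary, are listed by trimmed name as "movable" errors, invalidating the asset.
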EditorUtilityBlueprint /Game/NeverCook/EditorUtilities/GenericAssetReviewTool/BP/ReportExamples/Blueprint/Report_BuildingActor_ComponentMobility_Base
graphs [
 CheckIfBuildingActorClass {nodes [
   CheckIfBuildingActorClass!() -> (then, AssetData:Asset Data Structure) {is=title K2Node_FunctionEntry}
   > n1 = For Each Loop with Break!(
    Exec:Exec
    Array:Array of Object Soft Class References=Get BuildingActorClasses() -> (BuildingActorClasses:Array of Object Soft Class References)
    Break:Exec=@n4.then
   ) -> (LoopBody:Exec, Array Element:Object Soft Class Reference, Array Index:int, Completed:Exec)
   n2 = MakeSoftClassPath(PathString=GetTagValue(InAssetData=Get AssetData(), InTagName=MakeLiteralName(Value=NativeParentClass)))
   n3 = "Equal Exactly (String)"(
    A="To String (SoftClassReference)"!(SoftClassReference:"Object Soft Class Reference (by ref)"=To Soft Class Reference(SoftClassPath=@n2)) -> (ReturnValue:String)
    B="To String (SoftClassReference)"(SoftClassReference=@n1)
   )
   Return Node!(execute=@n1.Completed, IsBuildngActorClass:bool=@n3)
   > n4 = Branch(Condition=@n3)
  ]}
 EventGraph {links [], nodes []}
 GetBlueprintSmComponents {nodes [
   n1 = GetBlueprintSmComponents!() -> (then, AssetData:Asset Data Structure) {is=title K2Node_FunctionEntry}
   n2 = Is Valid Asset Data!(InAssetData:"Asset Data Structure (by ref)"=@n1) -> (ReturnValue:bool)
   > n3 = Branch(Condition=@n2)
   > n4 = GetBlueprintAsset(Object=GetAsset(InAssetData=@n1))
   > n5 = Gather Subobject Data For Blueprint(self=Get SubobjectDataSubsystem(), Context=@n4)
   > n6 = For Each Loop(Array:Array of Subobject Data Handle Structures=@n5) -> (Array Element:Subobject Data Handle Structure)
   n7 = Get l_StaticMeshComponents() -> (l_StaticMeshComponents:Array of Static Mesh Component Object References)
   n8 = "integer > integer"(A=Length(TargetArray:Array of Static Mesh Component Object References=@n7))
   n9 = Branch(execute=@n6.Completed, Condition=@n8)
   > Return Node!(execute, StaticMeshComponents:Array of Static Mesh Component Object References=@n7, SMCompFound:bool=@n8)
   Return Node(execute=@n9.else, SMCompFound=@n8)
   n10 = GetAssociatedObject!(Data:"Subobject Data Structure (by ref)"=GetData(DataHandle=@n6)) -> (ReturnValue:Object Reference)
   Is Valid(exec=@n6.LoopBody, InputObject=@n10)
   > n11 = Cast To StaticMeshComponent(Object=@n10)
   > n12 = Branch(Condition=Get bIsEditorOnly!(self:Actor Component Object Reference=@n11) -> (bIsEditorOnly:bool))
   Add(
    execute=@n12.else
    TargetArray:Array of Static Mesh Component Object References=@n7
    NewItem:"Static Mesh Component Object Reference (by ref)"=@n11
   ) {is=title K2Node_CallArrayFunction}
   Return Node(execute=@n3.else, SMCompFound=@n2)
  ]}
 RunFilterLogicOnAssetData {nodes [
   n1 = RunFilterLogicOnAssetData()
   > n2 = CheckIfBuildingActorClass!(execute, self, AssetData:Asset Data Structure=@n1) -> (then, IsBuildngActorClass:bool) {is=title K2Node_CallFunction}
   > n3 = Branch(Condition=@n2)
   > n4 = Sequence()
   > n5 = GetBlueprintSmComponents!(execute, self, AssetData:Asset Data Structure=Get AssetData()) -> (then, StaticMeshComponents:Array of Static Mesh Component Object References, SMCompFound:bool) {is=title K2Node_CallFunction}
   > n6 = Branch(Condition=@n5.SMCompFound)
   > n7 = For Each Loop(Array:Array of Static Mesh Component Object References=@n5) -> (Array Element:Static Mesh Component Object Reference)
   n8 = Get l_MovableSmComps() -> (l_MovableSmComps:Array of Strings)
   Branch(execute=@n7.Completed, Condition="integer > integer"(A=Length(TargetArray:Array of Strings=@n8)))
   > AddError(NewError=Movable static mesh components:)
   > n9 = For Each Loop(Array:Array of Strings=Get l_MovableSmComps()) -> (Array Element:String)
   > AddError(NewError=@n9, Depth=1)
   n10 = GetDisplayName(Object=@n7)
   Set l_SmCompName(
    execute=@n7.LoopBody
    l_SmCompName:String=Split(SourceString=Split(SourceString=@n10, InStr=., SearchDir=FromStart).RightS, InStr=_GEN_, SearchCase=CaseSensitive, SearchDir=FromStart)
   ) -> (Output_Get:String)
   > n11 = GetEditorProperty!(execute:Exec, Object:Object Reference=@n7, PropertyName:Name=Mobility) -> (then:Exec, PropertyValue:EComponentMobility Enum, ReturnValue:bool)
   > Branch(Condition="Not Equal (Enum)"!(A:EComponentMobility Enum=@n11, B:EComponentMobility Enum=Static) -> (ReturnValue:bool))
   > Add(TargetArray:Array of Strings=Get l_MovableSmComps(), NewItem:"String (by ref)"=Get l_SmCompName() -> (l_SmCompName:String))
   Return Node!(execute=@n6.else, Valid:bool, ReportData:Struct GART Report Data Structure)
   Return Node(
    execute=@n4.then_1
    Valid=IsEmpty(InString=GetCurrentErrorString!() -> (CurrentErrorString:String))
    ReportData=Make Struct GART Report Data(
     DisplayTitle_20_#1=Building Actor
     DisplayDescShort_29_#2="Contains moveable SM component(s)"
     DisplayColorOverride_40_#3="Lerp (LinearColor)"(A=(R=0.712,G=0,B=0.7552,A=1), Alpha=0)
    )
   )
   Return Node(execute=@n3.else)
  ]}
]
variables [BuildingActorClasses {
  type="array<Class>"
  default [
   &BuildingWall
   &BuildingFloor
   &BuildingStairs
   &BuildingRoof
   &BuildingProp
   /Script/FortniteGame.BuildingStaticGeo
   /Script/FortniteGame.BuildingPillar
   /Script/FortniteGame.BuildingCorner
  ]
 }]
