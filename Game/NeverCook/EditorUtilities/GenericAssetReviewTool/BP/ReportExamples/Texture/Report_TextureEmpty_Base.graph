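# Editor utility report for the Generic Asset Review Tool that flags Texture2D assets with empty or nearly uniform channels.
# RunFilterLogicOnAssetData calls CheckForEmptyChannel, which requires asset loading and skips normal maps via ShouldRun.
# It measures each channel's source min-to-max spread against MaxAlphaErrorTolerance, checking alpha only when compression settings and the hasAlphaChannel tag allow.
# Failing channels add errors, then set severity and description for all, some or possibly empty channels.
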EditorUtilityBlueprint /Game/NeverCook/EditorUtilities/GenericAssetReviewTool/BP/ReportExamples/Texture/Report_TextureEmpty_Base
graphs [
 CheckForEmptyChannel {nodes [
   n1 = CheckForEmptyChannel!() -> (then, AssetData:Asset Data Structure, AllowLoad:bool=~) {is=title K2Node_FunctionEntry}
   > Set l_assetData(l_assetData=@n1)
   > n2 = Branch(Condition=@n1.AllowLoad)
   > n3 = Cast To Texture2D(Object=GetAsset(InAssetData=Get l_assetData()))
   > Set l_texture(l_texture:Texture 2D Object Reference=@n3) -> (Output_Get:Texture 2D Object Reference)
   > Set l_completelyEmpty(l_completelyEmpty:bool=true) -> (Output_Get:bool)
   > n4 = Sequence!(execute) -> (then_0, then_1, then_2:Exec, then_3:Exec, then_4:Exec)
   > n5 = Branch(Condition=ShouldRun!(texture:Texture Object Reference=Get l_texture() -> (l_texture:Texture 2D Object Reference)) -> (ReturnValue:bool))
   Return Node!(execute=@n5.else, Valid:bool, ReportData:Struct GART Report Data Structure)
   Set l_hasAlpha(execute=@n4.then_1, l_hasAlpha:bool=DoesTextureHaveOptionalAlpha!(self2:Texture Object Reference=Get l_texture()) -> (ReturnValue:bool)) -> (Output_Get:bool)
   n6 = ComputeTextureSourceChannelMinMax!(execute=@n4.then_2, self:Texture Object Reference=Get l_texture()) -> (then, OutColorMin:Linear Color Structure, OutColorMax:Linear Color Structure, ReturnValue:bool)
   n7 = BreakColor(InColor=LinearColor - LinearColor(A:Linear Color Structure=@n6, B:Linear Color Structure=@n6.OutColorMax) -> (ReturnValue:Linear Color Structure) {is=title K2Node_PromotableOperator})
   > Set l_differenceSet(l_differenceSet:Vector 4 Structure=MakeVector4!(
     X:double="Absolute (Float)"(A=@n7)
     Y:double="Absolute (Float)"(A=@n7.G)
     Z:double="Absolute (Float)"(A=@n7.B)
     W:double="Absolute (Float)"(A=@n7.A)
    ) -> (ReturnValue:Vector 4 Structure)) -> (Output_Get:Vector 4 Structure)
   > n8 = For Loop(LastIndex=3)
   > Branch(Condition=OR Boolean(A=Get l_hasAlpha() -> (l_hasAlpha:bool), B="Not Equal (Integer)"(A=@n8, B=3)))
   n9 = BreakVector4!(InVec:"Vector 4 Structure (by ref)"=Get l_differenceSet() -> (l_differenceSet:Vector 4 Structure)) -> (X:double, Y:double, Z:double, W:double)
   > Set l_difference(l_difference:double=Select!(Option 0:double=@n9, Option 1:double=@n9.Y, Option 2:double=@n9.Z, Option 3:double=@n9.W, Index:int=@n8) -> (ReturnValue:double)) -> (Output_Get:double)
   > Set l_channel(l_channel:String=Select(Option 0:String=Red, Option 1:String=Green, Option 2:String=Blue, Option 3:String=Alpha, Index=@n8) -> (ReturnValue:String)) -> (Output_Get:String)
   > n10 = Branch(Condition="float <= float"(A=Get l_difference() -> (l_difference:double), B=Get MaxAlphaErrorTolerance() -> (MaxAlphaErrorTolerance:double)) -> (ReturnValue=~))
   n11 = Select!(
    Option 0:Text=MakeLiteralText(Value="INVTEXT(\" is very subtle and nearly empty.\")")
    Option 1:Text=MakeLiteralText(Value="INVTEXT(\" is empty.\")")
    Index:bool="Nearly Equal (Float)"(A=Get l_difference(), B=0, ErrorTolerance=0.001)
   ) -> (ReturnValue:Text)
   n12 = Get l_channel() -> (l_channel:String)
   > AddError(NewError="To String (Text)"(InText=Format Text!(Format="INVTEXT(\"{channel} is {reason}\")", channel:String=@n12, reason:Text=@n11) -> (Result)))
   n13 = Get l_difference()
   > Set l_anyEmptyChannels(l_anyEmptyChannels:bool=OR Boolean(A="Nearly Equal (Float)"(A=@n13, B=0, ErrorTolerance=0.001), B=Get l_anyEmptyChannels() -> (l_anyEmptyChannels:bool))) -> (Output_Get:bool)
   > n14 = Branch(Condition="Nearly Equal (Float)"(A=Get l_difference(), B=0, ErrorTolerance=0.001))
   Set l_completelyEmpty(execute=@n14.else, l_completelyEmpty=false)
   Set l_completelyEmpty(execute=@n10.else, l_completelyEmpty=false)
   n15 = Branch(execute=@n4.then_3, Condition=IsEmpty(InString=GetCurrentErrorString!() -> (CurrentErrorString:String)))
   n16 = Branch(execute=@n15.else, Condition=Get l_completelyEmpty() -> (l_completelyEmpty:bool))
   > SetSeverity!(execute, self:GART Report Base Object Reference, NewSeverity:Enum_GART_ReportSeverity Enum=NewEnumerator2, OnlyIfHigher:bool=true) -> (then)
   > Set l_descShort(l_descShort:String=All Empty Channels) -> (Output_Get:String)
   > Set l_valid(l_valid:bool=false) -> (Output_Get:bool)
   n17 = Branch(execute=@n16.else, Condition=Get l_anyEmptyChannels())
   > SetSeverity(NewSeverity=NewEnumerator2, OnlyIfHigher=true)
   > Set l_descShort(l_descShort=Some Empty Channels)
   > Set l_valid(l_valid=false)
   SetSeverity(execute=@n17.else, NewSeverity=NewEnumerator1, OnlyIfHigher=true)
   > Set l_descShort(l_descShort=Some Possibly Empty Channels)
   > Set l_valid(l_valid=false)
   Return Node(
    execute=@n4.then_4
    Valid=Get l_valid() -> (l_valid:bool)
    ReportData=Make Struct GART Report Data(
     DisplayTitle_20_#1=Texture
     DisplayDescShort_29_#2=Get l_descShort() -> (l_descShort:String)
     Severity_46_#3=NewEnumerator0
    )
   )
   Return Node(execute=[@n3.CastFailed, @n2.else])
   Vector_GetAbs!(A:Vector=vector - vector(
     A="To Vector (LinearColor)"!(InLinearColor:Linear Color Structure=@n6) -> (ReturnValue:Vector)
     B="To Vector (LinearColor)"(InLinearColor=@n6.OutColorMax)
    )) -> (ReturnValue:Vector)
   LinearColor - LinearColor(A=@n6, B=@n6.OutColorMax)
  ]}
 DoesTextureHaveOptionalAlpha {nodes [
   n1 = DoesTextureHaveOptionalAlpha!() -> (then, self2:Texture Object Reference) {is=title K2Node_FunctionEntry}
   n2 = GetTagValue(InAssetData=CreateAssetData(InAsset=@n1, bAllowBlueprintClass=false), InTagName=MakeLiteralName(Value=hasAlphaChannel))
   n3 = Get CompressionSettings(self=@n1)
   > Return Node!(
    execute
    ReturnValue:bool=AND Boolean!(
     A="Not Equal (Enum)"!(A:TextureCompressionSettings Enum=@n3, B:TextureCompressionSettings Enum=TC_HDR) -> (ReturnValue:bool)
     B="Not Equal (Enum)"(A=@n3, B=TC_Grayscale)
     C:bool="Not Equal (Enum)"(A=@n3, B=TC_Displacementmap)
     D:bool="Not Equal (Enum)"(A=@n3, B=TC_DistanceFieldFont)
     E:bool="Not Equal (Enum)"(A=@n3, B=TC_HalfFloat)
     F:bool="Not Equal (Enum)"(A=@n3, B=TC_Normalmap)
     G:bool="Not Equal (Enum)"(A=@n3, B=TC_HDR_Compressed)
     H:bool="Not Equal (Enum)"(A=@n3, B=TC_VectorDisplacementmap)
     I:bool="Equal Exactly (String)"(A=@n2, B=True)
     J:bool="Not Equal (Enum)"(A=@n3, B=TC_Normalmap)
    ) -> (ReturnValue)
   )
   Cast To Texture2D(Object:Wildcard)
  ]}
 EventGraph {links [], nodes []}
 RunFilterLogicOnAssetData {nodes [
   n1 = RunFilterLogicOnAssetData()
   > n2 = CheckForEmptyChannel!(execute, self, AssetData:Asset Data Structure=@n1, AllowLoad:bool=@n1.AllowLoading) -> (then, Valid:bool, ReportData:Struct GART Report Data Structure) {is=title K2Node_CallFunction}
   > Return Node!(execute, Valid:bool=@n2, ReportData:Struct GART Report Data Structure=@n2.ReportData)
  ]}
 ShouldRun {nodes [
   n1 = ShouldRun!() -> (then, texture:Texture Object Reference) {is=title K2Node_FunctionEntry}
   > Return Node!(execute, ReturnValue:bool="Not Equal (Enum)"(A=Get CompressionSettings(self=@n1), B=TC_Normalmap))
  ]}
]
variables [MaxAlphaErrorTolerance {type=number, default=0.1}]
